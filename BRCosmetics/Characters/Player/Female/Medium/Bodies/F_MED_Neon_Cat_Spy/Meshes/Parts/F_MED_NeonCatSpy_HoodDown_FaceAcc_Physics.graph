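# Physics asset for the Neon Cat Spy's hood-down face accessory, defining bodies for ponytail and lowered-hood secondary motion.
# The ponytail chain runs from a head-attached base into three hair ends.
# The hood chain hangs from spine_05, with a mid panel and two-segment cat ears branching off its top.
# Dynamic links have tight swing and twist limits, while small base boxes, head and clavicles use free angular constraints.
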PhysicsAsset /BRCosmetics/Characters/Player/Female/Medium/Bodies/F_MED_Neon_Cat_Spy/Meshes/Parts/F_MED_NeonCatSpy_HoodDown_FaceAcc_Physics
bodies {
 clavicle_l [{shapeType=Capsule, length=11.7179, radius=4.4982, center=<2.5623 -1.8369 0>, rotation=<rot 90 0 180>}]
 clavicle_r [{shapeType=Capsule, length=11.7179, radius=4.4982, center=<-2.5623 1.8369 0>, rotation=<rot 90 0 0>}]
 dyn_dropEar_01_l [{shapeType=Capsule, length=4.205, radius=2.1483, center=<1.7714 -0.2035 -0.2511>, rotation=<rot 90 0 0>}]
 dyn_dropEar_01_r [{shapeType=Capsule, length=3.0918, radius=2.125, center=<1.8415 0.0832 0.0077>, rotation=<rot 75.6261 -158.579 -158.1166>}]
 dyn_dropEar_02_l [{shapeType=Capsule, length=1.966, radius=1.676, center=<1.6883 -0.379 0.3239>, rotation=<rot 0 81.5506 -79.3317>}]
 dyn_dropEar_02_r [{shapeType=Capsule, length=1.967, radius=1.6702, center=<1.9338 -0.2036 -0.2828>, rotation=<rot 0 -97.5076 -77.7581>}]
 dyn_hairEnd_A [{shapeType=Capsule, length=3.057, radius=1.7713, center=<1.5854 0 0>, rotation=<rot 90 0 0>}]
 dyn_hairEnd_B [{shapeType=Capsule, length=3.057, radius=1.7713, center=<1.5854 0 0>, rotation=<rot 90 0 0>}]
 dyn_hairEnd_C [{shapeType=Capsule, length=3.057, radius=1.7713, center=<1.5854 0 0>, rotation=<rot 90 0 0>}]
 dyn_hoodDrop_Mid [{shapeType=Box, extentX=6.7916, extentY=4.2127, extentZ=9.8561, center=<2.9482 -0.3442 0>, rotation=<rot 0 -7.3299 0>}]
 dyn_hoodDrop_Top [{shapeType=Box, extentX=7.125, extentY=4.7226, extentZ=10.4826, center=<0.8302 2.6572 0>, rotation=<rot 0 73.6677 0>}]
 dyn_hoodDrop_Top_Base [{shapeType=Box, extentX=1.01, extentY=1.01, extentZ=1.01, rotation=<rot 0 -90 0>}]
 dyn_ponyTailMain [{shapeType=Capsule, length=3.633, radius=2.6976, center=<2.9742 0 0>, rotation=<rot 90 0 0>}]
 dyn_ponyTailMain_Base [{shapeType=Box, extentX=1.01, extentY=1.01, extentZ=1.01, rotation=<rot 0 -90 0>}]
 head [{shapeType=Capsule, length=3.1792, radius=3.3306, center=<0.4391 2.6051 0>, rotation=<rot 90 0 -21.3102>}]
 spine_05 [{shapeType=Capsule, length=17.1589, radius=8.1499}]
}
constraints [
 {
  bone1Name=clavicle_r
  bone2Name=spine_05
  swing1LimitDegrees=45
  swing1Motion=Free
  swing2LimitDegrees=45
  swing2Motion=Free
  twistLimitDegrees=45
  twistMotion=Free
 }
 {
  bone1Name=clavicle_l
  bone2Name=spine_05
  swing1LimitDegrees=45
  swing1Motion=Free
  swing2LimitDegrees=45
  swing2Motion=Free
  twistLimitDegrees=45
  twistMotion=Free
 }
 {bone1Name=dyn_ponyTailMain, bone2Name=dyn_ponyTailMain_Base, swing1LimitDegrees=16, swing2LimitDegrees=16, twistLimitDegrees=4}
 {bone1Name=dyn_hairEnd_B, bone2Name=dyn_ponyTailMain, swing1LimitDegrees=14, swing2LimitDegrees=14, twistLimitDegrees=12}
 {bone1Name=dyn_hairEnd_A, bone2Name=dyn_ponyTailMain, swing1LimitDegrees=14, swing2LimitDegrees=14, twistLimitDegrees=12}
 {bone1Name=dyn_hairEnd_C, bone2Name=dyn_ponyTailMain, swing1LimitDegrees=14, swing2LimitDegrees=14, twistLimitDegrees=12}
 {
  bone1Name=dyn_ponyTailMain_Base
  bone2Name=head
  swing1LimitDegrees=45
  swing1Motion=Free
  swing2LimitDegrees=45
  swing2Motion=Free
  twistLimitDegrees=45
  twistMotion=Free
 }
 {
  bone1Name=dyn_hoodDrop_Top_Base
  bone2Name=spine_05
  swing1LimitDegrees=45
  swing1Motion=Free
  swing2LimitDegrees=45
  swing2Motion=Free
  twistLimitDegrees=45
  twistMotion=Free
 }
 {bone1Name=dyn_hoodDrop_Top, bone2Name=dyn_hoodDrop_Top_Base, swing1LimitDegrees=25, swing2LimitDegrees=25, twistLimitDegrees=6}
 {bone1Name=dyn_hoodDrop_Mid, bone2Name=dyn_hoodDrop_Top, swing1LimitDegrees=25, swing2LimitDegrees=25, twistLimitDegrees=6}
 {bone1Name=dyn_dropEar_01_l, bone2Name=dyn_hoodDrop_Top, swing1LimitDegrees=12, swing2LimitDegrees=12, twistLimitDegrees=12}
 {bone1Name=dyn_dropEar_02_l, bone2Name=dyn_dropEar_01_l, swing1LimitDegrees=12, swing2LimitDegrees=12, twistLimitDegrees=12}
 {bone1Name=dyn_dropEar_01_r, bone2Name=dyn_hoodDrop_Top, swing1LimitDegrees=16, swing2LimitDegrees=16, twistLimitDegrees=12}
 {bone1Name=dyn_dropEar_02_r, bone2Name=dyn_dropEar_01_r, swing1LimitDegrees=12, swing2LimitDegrees=12, twistLimitDegrees=12}
 {
  bone1Name=head
  bone2Name=spine_05
  swing1LimitDegrees=45
  swing1Motion=Free
  swing2LimitDegrees=45
  swing2Motion=Free
  twistLimitDegrees=45
  twistMotion=Free
 }
]
